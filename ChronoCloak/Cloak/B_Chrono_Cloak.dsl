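; B_Chrono_Cloak is the Fortnite weapon Blueprint for the Chrono Cloak item, ending the cloak when the player switches weapons.
; On unequip and detach, it finds the GA_Chrono_Cloak_Manager instance on the instigator and calls RemoveCloak, only if the current weapon is valid and of another class.
; Comparing classes covers Changing Booth loadout refreshes that replace the weapon instance; the parent event runs only on that branch.
; CanApplyHolster always returns false.
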(event EventBeginPlay
  (|CallParentFunction) ; This node is disabled and will not be called.\nDrag off pins to build functionality.
)

(event EventTick
  (params (DeltaSeconds :type Float%20%28single-precision%29))
  (|CallParentFunction DeltaSeconds) ; This node is disabled and will not be called.\nDrag off pins to build functionality.
)

(event FortWeapon|EventOnUnEquippedAndDetached
  (bind _returnvalue (Game|GetInstigator))
  ;[1 0.1294 0.05967 0.5] Remove the ability when the active weapon changes, even if the loadout refreshed and caused the weapon to be destroyed and replaced with a new instance (e.g. from a Changing Booth).
  ; Without this, we don't end the ability when the player switches weapons after exiting a Changing Booth.
  (bind _asfort_pawn (Utilities|Casting|CastToFortPawn _returnvalue))
  (Utilities|IsValid (Weapon|GetCurrentWeapon _asfort_pawn)
    (:Is%20Valid
      (if (!= (Utilities|GetClass (Weapon|GetCurrentWeapon _asfort_pawn)) (Utilities|GetClass self))
        (bind _returnvalue_2 (Ability|GetPrimaryAbilityInstanceOfAbilitySystemComponent (Ability|GetFortAbilitySystemComponent _returnvalue) (Variables|Default|GetGA_Chrono_Cloak_Manager)))
        (bind _asga_chrono_cloak_manager (Utilities|Casting|CastToGA_Chrono_Cloak_Manager _returnvalue_2))
        (Class|GAChronoCloakManager|RemoveCloak _asga_chrono_cloak_manager)
        (|CallParentFunction)))))

(fn CanApplyHolster
  (params (HolsterId :type Name%20%28by%20ref%29 :default "None"))
  (return false))

(fn ConstructionScript
  (|CallParentFunction))
