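; Gameplay ability for exiting the Chrono Cloak, which clears the Nightmare Cloak gameplay cue from the owner.
; It runs the parent logic on both activation and end, then removes the cue, so the cue is cleared at both points.
; It also defines CanApplyHolster to always return false, so holster application is never permitted through this ability.
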(event Ability|EventActivateAbility
  (|CallParentFunction)
  (Ability|RemoveGameplayCueFromOwner (Variables|Default|GetTag_GC_NightmareCloak)))

(event Ability|EventOnEndAbility
  (params (bWasCancelled :type Boolean))
  (|CallParentFunction bWasCancelled)
  (Ability|RemoveGameplayCueFromOwner (Variables|Default|GetTag_GC_NightmareCloak)))

(fn CanApplyHolster
  (params (HolsterId :type Name%20%28by%20ref%29 :default "None"))
  (return false))
