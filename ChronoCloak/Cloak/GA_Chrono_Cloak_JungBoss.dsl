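; Gameplay ability for the JungBoss variant of the Chrono Cloak, layering activation rules over a parent ability that does the main work.
; Activation calls the parent, then fires CastMontageNotifyStart immediately, a noted hack against server montage cancellation while the activation key is held.
; CanActivateAbility fails as blocked on prevent-activation tags or a non-FortPawn avatar, otherwise deferring to the parent.
; Failures always allow the failed montage and add a CoveredInMud reason when that tag is granted.
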(event Ability|EventActivateAbility
  (|CallParentFunction)
  (CallFunction|CastMontageNotifyStart) ; Hack - to avoid server montage cancellation when holding left click (activation key)
)

(event Ability|EventOnEndAbility
  (params (bWasCancelled :type Boolean))
  (|CallParentFunction bWasCancelled) ; This node is disabled and will not be called.\nDrag off pins to build functionality.
)

(fn CanActivateAbility
  (params (ActorInfo :type Gameplay%20Ability%20Actor%20Info%20Structure) (Handle :type Gameplay%20Ability%20Spec%20Handle%20Structure))
  (bind (_owneractor _avataractor) (Utilities|Struct|BreakGameplayAbilityActorInfo ActorInfo))
  (bind _returnvalue (GameplayTags|HasAnyMatchingGameplayTags(Message) _avataractor (Variables|Default|GetTags_PreventActivation)))
  (if (not _returnvalue)
    (bind _asfort_pawn (Utilities|Casting|CastToFortPawn _avataractor)
      (:then
        (if true
          (bind (_relevanttags _returnvalue_2) (|CallParentFunction ActorInfo Handle))
          (return _relevanttags _returnvalue_2)))
      (:CastFailed
        _))
    (else
      _))
  (return (Variables|Default|GetTag_FailedReasonBlocked) false))

(fn OverrideFailedReason
  (params (FailedReason :type Gameplay%20Tag%20Container%20Structure%20%28by%20ref%29))
  (bind _overridenfailedreason (|CallParentFunction FailedReason))
  (if (GameplayTags|HasAnyMatchingGameplayTags (Ability|GetAbilitySystemComponentfromActorInfo) (GameplayTags|MakeGameplayTagContainerfromTag "(TagName=\"Granted.Mud.CoveredInMud\")"))
    (GameplayTags|AddGameplayTag _overridenfailedreason "(TagName=\"Ability.ActivationFailedReason.CoveredInMud\")" :!class /Script/GameplayTags.BlueprintGameplayTagLibrary)
    (else
      _))
  (return _overridenfailedreason))

(fn ShouldPlayFailedMontage
  (params (FailedReason :type Gameplay%20Tag%20Container%20Structure%20%28by%20ref%29))
  (return true))
