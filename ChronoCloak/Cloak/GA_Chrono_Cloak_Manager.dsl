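; Gameplay ability that manages the active Nightmare Cloak, adding its gameplay cue and periodically hiding the owner from AI.
; A looping timer, with its interval randomized at activation, reapplies the hide effect.
; Tasks watch unequip, emote, DBNO, death, locker, swimming and optional wet-tag changes, alongside a water-immersion check that can remove the cloak.
; Ending clears timers and the hide effect, then removes the granting effect by tag.
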(event Custom|RemoveCloak
  ;[1 1 1] Remove Cloak
  (Ability|RemoveGameplayEffectFromOwnerWithGrantedTags (Variables|Tags|GetTag_NightmareCloak_Active)) ; Removes granting effect, ending ability
)

(event Custom|StartGameplayEventManager
  ;[1 1 1] Gameplay Event Manager
  (bind _asynctaskproxy (Ability|Tasks|WaitGameplayTagRemove (Variables|Tags|GetTag_NightmareCloak_Equipped))
    (:then
      (Variables|TimerandTaskHandles|SetTask_WaitEquippedTagRemove _asynctaskproxy)
      (bind _asynctaskproxy_2 (Ability|Tasks|WaitGameplayTagAdd (Variables|Tags|GetTag_Granted_BuiltInEmote))
        (:then
          (Variables|Default|SetTask_WaitBuiltInEmoteTagAdd _asynctaskproxy_2)
          (bind _asynctaskproxy_3 (Ability|Tasks|WaitGameplayTagAdd (Variables|Tags|GetTag_GameplayStatus_DBNO))
            (:then
              (Variables|TimerandTaskHandles|SetTask_WaitDBNOTagAdd _asynctaskproxy_3)
              (bind _asynctaskproxy_4 (Ability|Tasks|WaitGameplayEvent (Variables|Tags|GetTag_GameplayEvent_Died))
                (:then
                  (Variables|TimerandTaskHandles|SetTask_WaitGameplayEvent_Died _asynctaskproxy_4)
                  (bind _asynctaskproxy_5 (Ability|Tasks|WaitGameplayTagAdd (Variables|Tags|GetTag_Locker))
                    (:then
                      (Variables|TimerandTaskHandles|SetTask_WaitLockerTagAdd _asynctaskproxy_5)
                      (bind _asynctaskproxy_6 (Ability|Tasks|WaitGameplayTagAdd (Variables|Tags|GetTag_Swimming)))
                      (Variables|TimerandTaskHandles|SetTask_WaitSwimmingTagAdd _asynctaskproxy_6)
                      (if (> (ScalableFloat|GetValueatLevel (Variables|Balance|GetRow_EnableWaterDepthChecks)) 0.0)
                        (bind _asynctaskproxy_7 (Ability|Tasks|WaitGameplayTagAdd (Variables|Tags|GetTag_Wet)))
                        (Variables|TimerandTaskHandles|SetTask_WaitWetTagAdd _asynctaskproxy_7)
                        (bind _asynctaskproxy_8 (Ability|Tasks|WaitGameplayTagRemove (Variables|Tags|GetTag_Wet)))
                        (Variables|TimerandTaskHandles|SetTask_WaitWetTagRemove _asynctaskproxy_8)))
                    (:Added
                      _)))
                (:EventReceived
                  _))
              (Ability|EndAbility))
            (:Added
              _)))
        (:Added
          _)))
    (:Removed
      _)))

(event Custom|HideFromAI
  ;[1 1 1] Hide From AI
  (Utilities|IsValid (Ability|GetAvatarActorfromActorInfo)
    (:Is%20Valid
      (bind _returnvalue (Ability|ApplyGameplayEffectToOwner (Variables|Default|GetGE_HideFromAI) 0))
      (Variables|Default|SetGEHandle_HideFromAI _returnvalue))
    (:Is%20Not%20Valid
      (Ability|EndAbility))))

(event Ability|EventActivateAbility
  ;[1 1 1] On Activate Ability
  (|CallParentFunction)
  (CallFunction|HidefromAI)
  (bind _returnvalue (Utilities|Time|SetTimerbyEvent (EventDispatchers|CreateEvent :!event Custom|HideFromAI) (+ (Math|Random|RandomFloatinRange (ScalableFloat|GetValueatLevel (Variables|Balance|GetRow_MinTimeBetweenHideFromAI)) (ScalableFloat|GetValueatLevel (Variables|Balance|GetRow_MaxTimeBetweenHideFromAI))) (ScalableFloat|GetValueatLevel (Variables|Balance|GetRow_DurationHideFromAI))) true))
  (Variables|TimerandTaskHandles|SetHandle_HideFromAITimer _returnvalue)
  (Ability|AddGameplayCueToOwner (Variables|Tags|GetTag_GC_NightmareCloak))
  (CallFunction|StartGameplayEventManager))

(event Ability|EventOnEndAbility
  (params (bWasCancelled :type Boolean))
  ;[1 1 1] On End Ability
  (|CallParentFunction bWasCancelled)
  (Utilities|Time|ClearandInvalidateTimerbyHandle (Variables|TimerandTaskHandles|GetHandle_HideFromAITimer))
  (Utilities|IsValid (Ability|GetAvatarActorfromActorInfo)
    (:Is%20Valid
      (Ability|RemoveGameplayEffectFromOwnerWithHandle (Variables|Default|GetGEHandle_HideFromAI))
      (GameplayTasks|EndTask (Variables|TimerandTaskHandles|GetTask_WaitEquippedTagRemove))
      (Utilities|Time|ClearandInvalidateTimerbyHandle (Variables|TimerandTaskHandles|GetHandle_CheckWaterDepthTimer))
      (CallFunction|RemoveCloak))))

(event Custom|CheckWaterDepth
  (bind _returnvalue (Ability|GetAvatarActorfromActorInfo))
  ;[1 1 1] Check Water Depth
  (Utilities|IsValid _returnvalue
    (:Is%20Valid
      (bind _asfort_pawn (Utilities|Casting|CastToFortPawn _returnvalue)
        (:then
          (if (> (WaterBody|GetWaterImmersionDepth _asfort_pawn) (ScalableFloat|GetValueatLevel (Variables|Balance|GetRow_MaxAllowedWaterImmersionDepth)))
            (CallFunction|RemoveCloak)))
        (:CastFailed
          _)))
    (:Is%20Not%20Valid
      _))
  (Ability|EndAbility))
